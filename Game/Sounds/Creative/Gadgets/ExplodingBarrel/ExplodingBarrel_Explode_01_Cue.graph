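# Sound cue for the Creative exploding barrel gadget's explosion, layering near and distant sounds after a short random delay.
# The near layer mixes a random flintlock ball explosion with a quieter grenade blast under RPG explosion attenuation.
# The distant layer mixes a sharp distant boom with a random TNT tail, distance-crossfaded to play only at mid-to-far range.
# Each layer has pitch variation and its own sound class, and a concurrency override limits simultaneous voices.
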SoundCue /Game/Sounds/Creative/Gadgets/ExplodingBarrel/ExplodingBarrel_Explode_01_Cue
root=Delay_3
nodes [
 {
  id=Attenuation_8
  children [Mixer_6]
  properties {attenuationSettings=/Game/Sounds/Attenuation/Fort_Explosion_RPG_Att, attenuationOverrides {falloffDistance=1600}}
 }
 {
  id=Attenuation_9
  children [Modulator_9]
  properties {
   attenuationSettings=/Game/Sounds/Attenuation/Fort_Weapon_Projectile_Distant_Barrel_Att
   attenuationOverrides {falloffDistance=3000, attenuationShapeExtents {x=1000}}
  }
 }
 {id=Delay_3, children [Mixer_5], properties {delayMax=0.12, delayMin=0.02}}
 {
  id=DistanceCrossFade_3
  children [Attenuation_9, null]
  properties {crossFadeInput [
    {fadeInDistanceEnd=3000, fadeInDistanceStart=1500, fadeOutDistanceEnd=12000, fadeOutDistanceStart=10000, volume=1}
    {fadeInDistanceEnd=0, fadeInDistanceStart=0, fadeOutDistanceEnd=0, fadeOutDistanceStart=0, volume=1}
   ]}
 }
 {id=Mixer_0, children [WavePlayer_13, Random_0], properties {inputVolume [1,1]}}
 {id=Mixer_5, children [SoundClass_9,SoundClass_8], properties {inputVolume [0.85,0.7]}}
 {id=Mixer_6, children [Random_2, WavePlayer_12], properties {inputVolume [1,0.5]}}
 {id=Modulator_10, children [Attenuation_8], properties {pitchMax=1.1, pitchMin=0.9}}
 {id=Modulator_9, children [Mixer_0], properties {pitchMax=1.1, pitchMin=0.9}}
 {id=Random_0, children [WavePlayer_0,WavePlayer_1], properties {weights [1,1]}}
 {id=Random_2, children [WavePlayer_14, WavePlayer_15], properties {weights [1,1]}}
 {
  id=SoundClass_8
  children [DistanceCrossFade_3]
  properties {soundClassOverride=&Fort_WeaponRanged_Fire_Distant}
 }
 {
  id=SoundClass_9
  children [Modulator_10]
  properties {soundClassOverride=&Fort_WeaponRanged_Fire_3P_PSM}
 }
 {id=WavePlayer_0, properties {soundWaveAssetPtr=/Game/Sounds/Fort_GamePlay_Sounds/Explosives/TnT/TNT_Explode_Distant_02}}
 {id=WavePlayer_1, properties {soundWaveAssetPtr=/Game/Sounds/Fort_GamePlay_Sounds/Explosives/TnT/TNT_Explode_Distant_03}}
 {id=WavePlayer_12, properties {soundWaveAssetPtr=/Game/Sounds/Fort_GamePlay_Sounds/Explosives/Explosion_GrenadeProjectile_01_E}}
 {id=WavePlayer_13, properties {soundWaveAssetPtr=/Game/Sounds/Fort_GamePlay_Sounds/Explosives/explosion_sharp_distant_01}}
 {id=WavePlayer_14, properties {soundWaveAssetPtr=/Game/Sounds/Fort_Weapons/FlintLock/Flintlock_Launcher_BallExplode_01}}
 {id=WavePlayer_15, properties {soundWaveAssetPtr=/Game/Sounds/Fort_Weapons/FlintLock/Flintlock_Launcher_BallExplode_02}}
]
properties {
 bOverrideConcurrency=true
 maxDistance=50000
 soundClassObject=/Game/Sounds/FortSoundClassesAndModes/Fort_Explosives_Large_PSM
 volumeMultiplier=0.762
 attenuationOverrides {falloffDistance=20000}
 concurrencyOverrides {maxCount=2, voiceStealReleaseTime=0.5}
 concurrencySet [/Game/Sounds/Concurrency_Rules/OneMaximum_Owner_StopOldest]
}
