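# Sound cue for the electric exploding barrel's detonation, played after a brief fixed delay as a mix of near and quieter distant layers.
# The near layer combines a propane tank explosion with a quieter, slightly delayed random shock tower zap.
# The distant layer pairs a sharp distant explosion with a random distant TNT blast.
# Each layer has its own attenuation and sound class, with pitch modulation for variation and overridden concurrency capping simultaneous instances.
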SoundCue /Game/Sounds/Creative/Gadgets/ExplodingBarrel/ExplodingBarrel_Explode_Electric_Cue
root=Delay_3
nodes [
 {
  id=Attenuation_4
  children [Mixer_1]
  properties {
   attenuationSettings=/Game/Sounds/Attenuation/Fort_Explosion_RPG_Att
   attenuationOverrides {falloffDistance=5106.2798, stereoSpread=5084.1587, attenuationShapeExtents {x=2554.7803}}
  }
 }
 {
  id=Attenuation_9
  children [Modulator_9]
  properties {
   attenuationSettings=/Game/Sounds/Attenuation/Fort_Weapon_Projectile_Distant_Barrel_Att
   attenuationOverrides {falloffDistance=3000, attenuationShapeExtents {x=1000}}
  }
 }
 {id=Delay_0, children [Random_1], properties {delayMax=0.05, delayMin=0.02}}
 {id=Delay_3, children [Mixer_5], properties {delayMax=0.02, delayMin=0.02}}
 {id=Mixer_0, children [WavePlayer_13, Random_0], properties {inputVolume [1,1]}}
 {id=Mixer_1, children [Modulator_1,Modulator_0], properties {inputVolume [1.5,0.5]}}
 {id=Mixer_5, children [SoundClass_3,SoundClass_8], properties {inputVolume [1,0.4]}}
 {id=Modulator_0, children [Delay_0]}
 {id=Modulator_1, children [WavePlayer_0], properties {pitchMax=1.1}}
 {id=Modulator_9, children [Mixer_0], properties {pitchMax=1.1, pitchMin=0.9}}
 {id=Random_0, children [WavePlayer_1,WavePlayer_2], properties {weights [1,1]}}
 {id=Random_1, children [WavePlayer_9, WavePlayer_10, WavePlayer_11], properties {weights [1,1,1]}}
 {
  id=SoundClass_3
  children [Attenuation_4]
  properties {soundClassOverride=&Fort_WeaponRanged_Fire_3P_PSM}
 }
 {
  id=SoundClass_8
  children [Attenuation_9]
  properties {soundClassOverride=&Fort_WeaponRanged_Fire_Distant}
 }
 {id=WavePlayer_0, properties {soundWaveAssetPtr=/Game/Sounds/Fort_Destructibles/Propane_Tank/Destructible_Propane_Tank_Explode_01}}
 {id=WavePlayer_1, properties {soundWaveAssetPtr=/Game/Sounds/Fort_GamePlay_Sounds/Explosives/TnT/TNT_Explode_Distant_02}}
 {id=WavePlayer_10, properties {soundWaveAssetPtr=/Game/Sounds/Fort_Gadgets/ShockTower/shocktower_fire_02}}
 {id=WavePlayer_11, properties {soundWaveAssetPtr=/Game/Sounds/Fort_Gadgets/ShockTower/shocktower_fire_03}}
 {id=WavePlayer_13, properties {soundWaveAssetPtr=/Game/Sounds/Fort_GamePlay_Sounds/Explosives/explosion_sharp_distant_01}}
 {id=WavePlayer_2, properties {soundWaveAssetPtr=/Game/Sounds/Fort_GamePlay_Sounds/Explosives/TnT/TNT_Explode_Distant_03}}
 {id=WavePlayer_9, properties {soundWaveAssetPtr=/Game/Sounds/Fort_Gadgets/ShockTower/shocktower_fire_01}}
]
properties {bOverrideConcurrency=true, maxDistance=50000, concurrencyOverrides {maxCount=2, volumeScale=0.8762}}
