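# Sound cue for the ice exploding barrel's explosion among the Creative gadgets.
# After a short random delay, it mixes near and distant branches, each with its own attenuation and sound class override.
# The near branch layers randomized Flintlock ball and Chiller grenade explosions with volume and pitch variation.
# The distant branch distance-crossfades random distant Chiller explosions in and out, while the cue's concurrency override prevents new instances at its limit.
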SoundCue /Game/Sounds/Creative/Gadgets/ExplodingBarrel/ExplodingBarrel_Explode_Ice_Cue
root=Delay_3
nodes [
 {
  id=Attenuation_2
  children [Mixer_6]
  properties {attenuationSettings=/Game/Sounds/Attenuation/Fort_Explosion_RPG_Att, attenuationOverrides {falloffDistance=1600}}
 }
 {
  id=Attenuation_3
  children [Random_5]
  properties {
   attenuationSettings=/Game/Sounds/Attenuation/Fort_Weapon_Projectile_Distant_Barrel_Att
   attenuationOverrides {falloffDistance=3000, attenuationShapeExtents {x=1000}}
  }
 }
 {id=Delay_3, children [Mixer_5], properties {delayMax=0.12, delayMin=0.02}}
 {
  id=DistanceCrossFade_3
  children [Attenuation_3, null]
  properties {crossFadeInput [
    {fadeInDistanceEnd=3000, fadeInDistanceStart=1500, fadeOutDistanceEnd=12000, fadeOutDistanceStart=10000, volume=1}
    {fadeInDistanceEnd=0, fadeInDistanceStart=0, fadeOutDistanceEnd=0, fadeOutDistanceStart=0, volume=1}
   ]}
 }
 {id=Mixer_5, children [SoundClass_9,SoundClass_8], properties {inputVolume [0.85,0.7]}}
 {id=Mixer_6, children [Modulator_2,Modulator_12], properties {inputVolume [0.3,1]}}
 {id=Modulator_10, children [Attenuation_2], properties {pitchMax=1.1, pitchMin=0.9}}
 {id=Modulator_12, children [Random_0], properties {volumeMax=1.2, volumeMin=1.1}}
 {id=Modulator_2, children [Random_3], properties {pitchMax=1.2, volumeMax=1.2, volumeMin=1.1}}
 {id=Random_0, children [WavePlayer_20, WavePlayer_21, WavePlayer_22], properties {weights [1,1,1]}}
 {id=Random_3, children [WavePlayer_18, WavePlayer_19], properties {weights [1,1]}}
 {id=Random_5, children [WavePlayer_11, WavePlayer_16, WavePlayer_17], properties {weights [1,1,1]}}
 {
  id=SoundClass_8
  children [DistanceCrossFade_3]
  properties {soundClassOverride=&Fort_WeaponRanged_Fire_Distant}
 }
 {
  id=SoundClass_9
  children [Modulator_10]
  properties {soundClassOverride=&Fort_WeaponRanged_Fire_3P_PSM}
 }
 {id=WavePlayer_11, properties {soundWaveAssetPtr=/Game/Sounds/Fort_Weapons/Grenades/Chiller/ChillGrenade_Explode_Distant_01}}
 {id=WavePlayer_16, properties {soundWaveAssetPtr=/Game/Sounds/Fort_Weapons/Grenades/Chiller/ChillGrenade_Explode_Distant_02}}
 {id=WavePlayer_17, properties {soundWaveAssetPtr=/Game/Sounds/Fort_Weapons/Grenades/Chiller/ChillGrenade_Explode_Distant_03}}
 {id=WavePlayer_18, properties {soundWaveAssetPtr=/Game/Sounds/Fort_Weapons/FlintLock/Flintlock_Launcher_BallExplode_01}}
 {id=WavePlayer_19, properties {soundWaveAssetPtr=/Game/Sounds/Fort_Weapons/FlintLock/Flintlock_Launcher_BallExplode_02}}
 {id=WavePlayer_20, properties {soundWaveAssetPtr=/Game/Sounds/Fort_Weapons/Grenades/Chiller/ChillGrenade_Explode_01}}
 {id=WavePlayer_21, properties {soundWaveAssetPtr=/Game/Sounds/Fort_Weapons/Grenades/Chiller/ChillGrenade_Explode_02}}
 {id=WavePlayer_22, properties {soundWaveAssetPtr=/Game/Sounds/Fort_Weapons/Grenades/Chiller/ChillGrenade_Explode_03}}
]
properties {
 attenuationSettings=/Game/Sounds/Attenuation/Fort_Explosion_Far_StW_Att
 bOverrideConcurrency=true
 maxDistance=50000
 soundClassObject=/Game/Sounds/FortSoundClassesAndModes/Fort_Explosives_Large_PSM
 volumeMultiplier=0.8491
 concurrencyOverrides {resolutionRule=PreventNew}
}
