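# Sound cue for the shockwave exploding barrel's explosion in Creative, built from shockwave grenade and flintlock launcher samples.
# A short random delay feeds a mix of close and distant branches, each with its own sound class and attenuation.
# The close branch layers three randomized sample sets, while the distant branch plays random distant variants through a distance crossfade.
# Concurrency override caps simultaneous instances at two, stopping the oldest.
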SoundCue /Game/Sounds/Creative/Gadgets/ExplodingBarrel/ExplodingBarrel_Explode_Shockwave_Cue
root=Delay_3
nodes [
 {
  id=Attenuation_2
  children [Mixer_1]
  properties {attenuationSettings=/Game/Sounds/Attenuation/Fort_Explosion_RPG_Att, attenuationOverrides {falloffDistance=1600}}
 }
 {
  id=Attenuation_3
  children [Modulator_13]
  properties {
   attenuationSettings=/Game/Sounds/Attenuation/Fort_Weapon_Projectile_Distant_Barrel_Att
   attenuationOverrides {falloffDistance=3000, attenuationShapeExtents {x=1000}}
  }
 }
 {id=Delay_3, children [Mixer_5], properties {delayMax=0.12, delayMin=0.02}}
 {
  id=DistanceCrossFade_3
  children [Attenuation_3, null]
  properties {crossFadeInput [
    {fadeInDistanceEnd=3000, fadeInDistanceStart=1500, fadeOutDistanceEnd=12000, fadeOutDistanceStart=10000, volume=1}
    {fadeInDistanceEnd=0, fadeInDistanceStart=0, fadeOutDistanceEnd=0, fadeOutDistanceStart=0, volume=1}
   ]}
 }
 {id=Mixer_1, children [Random_0,Random_2,Random_1], properties {inputVolume [1,0.5,0.1]}}
 {id=Mixer_5, children [SoundClass_9,SoundClass_8], properties {inputVolume [0.85,0.7]}}
 {id=Modulator_10, children [Attenuation_2]}
 {id=Modulator_13, children [Random_8], properties {volumeMax=1.1, volumeMin=1}}
 {id=Random_0, children [WavePlayer_0,WavePlayer_1,WavePlayer_2], properties {weights [1,1,1]}}
 {id=Random_1, children [WavePlayer_8,WavePlayer_9], properties {weights [1,1]}}
 {id=Random_2, children [WavePlayer_11, WavePlayer_12, WavePlayer_13], properties {weights [1,1,1]}}
 {id=Random_8, children [WavePlayer_14, WavePlayer_15, WavePlayer_16], properties {weights [1,1,1]}}
 {
  id=SoundClass_8
  children [DistanceCrossFade_3]
  properties {soundClassOverride=&Fort_WeaponRanged_Fire_Distant}
 }
 {
  id=SoundClass_9
  children [Modulator_10]
  properties {soundClassOverride=&Fort_WeaponRanged_Fire_3P_PSM}
 }
 {id=WavePlayer_0, properties {soundWaveAssetPtr=/Game/Sounds/Fort_GamePlay_Sounds/Explosives/ShockwaveGrenade/Grenade_Shockwave_Explode01}}
 {id=WavePlayer_1, properties {soundWaveAssetPtr=/Game/Sounds/Fort_GamePlay_Sounds/Explosives/ShockwaveGrenade/Grenade_Shockwave_Explode02}}
 {
  id=WavePlayer_11
  properties {soundWaveAssetPtr=/Game/Sounds/Fort_GamePlay_Sounds/Explosives/ShockwaveGrenade/Grenade_Shockwave_ExplodeLayer01}
 }
 {
  id=WavePlayer_12
  properties {soundWaveAssetPtr=/Game/Sounds/Fort_GamePlay_Sounds/Explosives/ShockwaveGrenade/Grenade_Shockwave_ExplodeLayer02}
 }
 {
  id=WavePlayer_13
  properties {soundWaveAssetPtr=/Game/Sounds/Fort_GamePlay_Sounds/Explosives/ShockwaveGrenade/Grenade_Shockwave_ExplodeLayer03}
 }
 {
  id=WavePlayer_14
  properties {soundWaveAssetPtr=/Game/Sounds/Fort_GamePlay_Sounds/Explosives/ShockwaveGrenade/Grenade_Shockwave_Explode_Distant01}
 }
 {
  id=WavePlayer_15
  properties {soundWaveAssetPtr=/Game/Sounds/Fort_GamePlay_Sounds/Explosives/ShockwaveGrenade/Grenade_Shockwave_Explode_Distant02}
 }
 {
  id=WavePlayer_16
  properties {soundWaveAssetPtr=/Game/Sounds/Fort_GamePlay_Sounds/Explosives/ShockwaveGrenade/Grenade_Shockwave_Explode_Distant03}
 }
 {id=WavePlayer_2, properties {soundWaveAssetPtr=/Game/Sounds/Fort_GamePlay_Sounds/Explosives/ShockwaveGrenade/Grenade_Shockwave_Explode03}}
 {id=WavePlayer_8, properties {soundWaveAssetPtr=/Game/Sounds/Fort_Weapons/FlintLock/Flintlock_Launcher_BallExplode_01}}
 {id=WavePlayer_9, properties {soundWaveAssetPtr=/Game/Sounds/Fort_Weapons/FlintLock/Flintlock_Launcher_BallExplode_02}}
]
properties {
 bOverrideConcurrency=true
 maxDistance=50000
 soundClassObject=/Game/Sounds/FortSoundClassesAndModes/Fort_Explosives_Large_PSM
 volumeMultiplier=0.616
 attenuationOverrides {distanceAlgorithm=NaturalSound, falloffDistance=3800, lPFRadiusMax=3000, lPFRadiusMin=2000, attenuationShapeExtents {x=2200}}
 concurrencyOverrides {maxCount=2, resolutionRule=StopOldest}
}
