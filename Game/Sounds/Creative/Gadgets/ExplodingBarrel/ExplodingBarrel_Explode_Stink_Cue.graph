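# Sound cue for the explosion of a stink-type exploding barrel gadget in Creative.
# A mixer layers three random pools: stink barrel explosions, glass breaks and reused vending machine metal impacts.
# The glass layer has its own volume modulation and a boosted mix level, and a root modulator wraps the whole blend.
# Playback uses the smoke grenade attenuation and the ranged weapon sound class.
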SoundCue /Game/Sounds/Creative/Gadgets/ExplodingBarrel/ExplodingBarrel_Explode_Stink_Cue
root=Modulator_0
nodes [
 {id=Mixer_1, children [Random_1,Modulator_1,Random_2], properties {inputVolume [1,1.3,1]}}
 {id=Modulator_0, children [Mixer_1]}
 {id=Modulator_1, children [Random_0], properties {volumeMax=1.2, volumeMin=0.8}}
 {id=Random_0, children [WavePlayer_1,WavePlayer_2], properties {weights [1,1]}}
 {id=Random_1, children [WavePlayer_0,WavePlayer_3,WavePlayer_4], properties {weights [1,1,1]}}
 {id=Random_2, children [WavePlayer_5,WavePlayer_6,WavePlayer_7], properties {weights [1,1,1]}}
 {id=WavePlayer_0, properties {soundWaveAssetPtr=/Game/Athena/Sounds/Weapons/Stink_Grenade/StinkBarrel_Explode_01}}
 {id=WavePlayer_1, properties {soundWaveAssetPtr=/Game/Athena/Sounds/Weapons/Stink_Grenade/StinkBarrel_Glass_01}}
 {id=WavePlayer_2, properties {soundWaveAssetPtr=/Game/Athena/Sounds/Weapons/Stink_Grenade/StinkBarrel_Glass_02}}
 {id=WavePlayer_3, properties {soundWaveAssetPtr=/Game/Athena/Sounds/Weapons/Stink_Grenade/StinkBarrel_Explode_02}}
 {id=WavePlayer_4, properties {soundWaveAssetPtr=/Game/Athena/Sounds/Weapons/Stink_Grenade/StinkBarrel_Explode_03}}
 {
  id=WavePlayer_5
  properties {soundWaveAssetPtr=/Game/Sounds/Fort_Impact_Sounds/Demolition/VendingMachine/Metal_VendMachine_ImpactFinal_mix_03}
 }
 {id=WavePlayer_6, properties {soundWaveAssetPtr=/Game/Sounds/Fort_Impact_Sounds/Demolition/VendingMachine/Metal_VendMachine_ImpactFinal03}}
 {
  id=WavePlayer_7
  properties {soundWaveAssetPtr=/Game/Sounds/Fort_Impact_Sounds/Demolition/VendingMachine/Metal_VendMachine_ImpactFinal_mix_01}
 }
]
properties {
 attenuationSettings=/Game/Sounds/Attenuation/Fort_Grenade_Smoke_Att
 maxDistance=6800
 soundClassObject=&Fort_WeaponRanged
 volumeMultiplier=1.15
}
